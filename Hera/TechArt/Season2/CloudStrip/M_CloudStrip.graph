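# Translucent unlit Nanite material for cloud strips, widened along a spline into view-facing ribbons by vertex offset math.
# Color blends sun-direction halo tint with a background cubemap, using the low-altitude clouds parameter collection.
# Opacity combines two panning noise samples, edge falloff, vertex red and named reroutes; Fresnel and camera-height fade are computed per vertex.
# Strips above the camera get pushed far vertically; a sphere subtraction mask sits outside the opacity chain.
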Material /Hera/TechArt/Season2/CloudStrip/M_CloudStrip
graph {
 properties {
  bAutomaticallySetUsageInEditor=false
  bUsedWithNanite=true
  blendMode=BLEND_Translucent
  previewMesh=/Engine/Transient.MaterialEditorMeshComponent_2
  shadingModel=MSM_Unlit
  usedShadingModels=Unlit
 }
 nodes [
  n1 = ComponentMask(None=TextureCoordinate()) {g=true}
  n2 = Transform(None=Constant3Vector() {constant=<rgba 1 0 0 0>})
  n3 = WorldPosition() {worldPositionShaderOffset=WPT_CameraRelative}
  n4 = Normalize(VectorInput=CrossProduct(A=@n2, B=Normalize(VectorInput=@n3.XYZ)))
  n5 = Floor(None=ComponentMask(None=TextureCoordinate()) {r=true})
  n6 = Multiply(
   A=Multiply(A=Add(A=@n1) {constB=-0.5}, B=@n4)
   B=LinearInterpolate(
    A=ScalarParameter() {defaultValue=14500, parameterName=Spline Width}
    B=ScalarParameter() {defaultValue=10000, parameterName=Spline Width Thin}
    Alpha=Saturate(None=@n5)
   )
  ) {constB=3000}
  n7 = Normalize(VectorInput=Add(A=@n6, B=WorldPosition() {worldPositionShaderOffset=WPT_CameraRelative}.XYZ))
  n8 = ComponentMask(None=ActorPositionWS() {originType=CameraRelative}) {b=true}
  Add(
   A=Add(A=@n6, B=Multiply(A=@n7) {constB=1000})
   B=Multiply(A=Saturate(None=@n8), B=Constant3Vector() {constant=<rgba 0 0 -100000 0>}) {constB=-100000}
  )
  n9 = ComponentMask(None=CollectionParameter() {collection=/Game/Environments/Helios/TechArt/LowAltitudeClouds/V2/MPC_LowAltitudeClouds_V2, parameterName=ColorLight}) {b=true, g=true, r=true}
  n10 = ComponentMask(None=CollectionParameter() {collection=&MPC_LowAltitudeClouds_V2, parameterName=ColorHalo}) {b=true, g=true, r=true}
  n11 = CameraVectorWS()
  n12 = SkyAtmosphereLightDirection()
  n13 = Saturate(None=Multiply(A=DotProduct(A=@n11, B=@n12)) {constB=-1})
  n14 = Multiply(A=@n13, B=@n13)
  n15 = Multiply(A=Multiply(A=@n14, B=@n14)) {constB=3}
  n16 = TextureSample(UVs=Multiply(A=CameraVectorWS()) {constB=-1}) {
   samplerType=SAMPLERTYPE_LinearColor
   texture=/Game/Environments/Helios/TechArt/LowAltitudeClouds/V2/T_LowAltitudeClouds_BackgroundCubemap_Tex
  }
  n17 = ComponentMask(None=CollectionParameter() {collection=&MPC_LowAltitudeClouds_V2, parameterName=ColorLight}) {b=true, g=true, r=true}
  LinearInterpolate(A=Multiply(A=LinearInterpolate(A=@n9, B=@n10, Alpha=@n15)) {constB=2}, B=Multiply(A=Multiply(A=@n16.RGB, B=@n17)) {constB=2.5})
  n18 = TextureCoordinate()
  n19 = ComponentMask(None=@n18) {r=true}
  n20 = Add(A=Multiply(A=@n19) {constB=3}, B=Multiply(A=Time()) {constB=0.03})
  n21 = Add(A=Add(A=Multiply(A=@n19) {constB=0.25}, B=ComponentMask(None=@n18) {g=true}), B=Multiply(A=Time()) {constB=0.03})
  n22 = TextureCoordinate()
  n23 = ComponentMask(None=@n22) {r=true}
  n24 = Add(A=Multiply(A=@n23) {constB=-2.56}, B=Multiply(A=Time()) {constB=0.01})
  n25 = ComponentMask(None=@n22) {g=true}
  n26 = Add(A=Subtract(A=OneMinus(None=@n25), B=Multiply(A=@n23) {constB=0.14}), B=Multiply(A=Time()) {constB=0.04})
  n27 = Multiply(
   A=TextureSample(UVs=AppendVector(A=@n20, B=@n21)) {samplerType=SAMPLERTYPE_LinearGrayscale, texture=/Hermes/Techart/CloudStrip/T_CloudStrip}.R
   B=TextureSample(UVs=AppendVector(A=@n24, B=@n26)) {samplerType=SAMPLERTYPE_LinearGrayscale, texture=&T_CloudStrip}.R
  )
  n28 = ComponentMask(None=TextureCoordinate()) {g=true}
  n29 = Multiply(A=Abs(None=Add(A=@n28) {constB=-0.5})) {constB=2}
  n30 = OneMinus(None=Multiply(A=@n29, B=@n29))
  n31 = Multiply(A=Multiply(A=Multiply(A=@n27, B=@n30), B=VertexColor().R), B=NamedRerouteUsage())
  n32 = Multiply(A=Multiply(A=@n31, B=@n31)) {constB=1.2}
  n33 = NamedRerouteUsage()
  Multiply(A=Multiply(A=Multiply(A=@n32, B=@n33), B=ScalarParameter() {defaultValue=0.7, parameterName=Opacity}) {constB=0.4}, B=Constant() {r=1})
  n34 = CrossProduct(A=@n4, B=@n2)
  n35 = Normalize(VectorInput=WorldPosition() {worldPositionShaderOffset=WPT_CameraRelative}.XYZ)
  NamedRerouteDeclaration(None=VertexInterpolator(VS=DotProduct(A=@n34, B=@n35)).PS) {name=Fresnel, nodeColor=<rgba 0 0.0625 1 1>}
  n36 = Multiply(A=ComponentMask(None=ActorPositionWS() {originType=CameraRelative}) {b=true}) {constB=-0.0001}
  NamedRerouteDeclaration(None=VertexInterpolator(VS=Saturate(None=@n36)).PS) {name=Height Fade, nodeColor=<rgba 0.5703 1 0 1>}
  n37 = VectorParameter() {parameterName=Subtraction Mask Local Pos, defaultValue=<rgba 0 -33700 0 1>}
  OneMinus(None=SphereMask(
    A=WorldPosition().XYZ
    B=TransformPosition(None=@n37.RGB) {transformType=TRANSFORMPOSSOURCE_World}
    Radius=ScalarParameter() {defaultValue=40000, parameterName=Subtraction Mask Radius}
    Hardness=ScalarParameter() {defaultValue=0.4, parameterName=Subtraction Mask Hardness}
   ))
 ]
}
